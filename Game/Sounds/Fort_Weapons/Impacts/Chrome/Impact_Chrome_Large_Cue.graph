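# Sound cue for large chrome bullet impacts in the impacts sound class, using bullet-impact attenuation and rifle-impact concurrency.
# A short random delay feeds a quality-level switch with two shared branches.
# The first two quality slots pick randomly from nine waves, while the last two use a reduced set of three.
# Each branch applies the same random pitch and volume modulation to vary repeated hits.
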SoundCue /Game/Sounds/Fort_Weapons/Impacts/Chrome/Impact_Chrome_Large_Cue
root=Delay_0
nodes [
 {id=Delay_0, children [QualityLevel_1], properties {delayMax=0.13, delayMin=0.07}}
 {id=Modulator_1, children [Random_3], properties {pitchMax=1.25, pitchMin=1, volumeMax=1.1, volumeMin=0.8}}
 {id=Modulator_2, children [Random_2], properties {pitchMax=1.25, pitchMin=1, volumeMax=1.1, volumeMin=0.8}}
 {id=QualityLevel_1, children [Modulator_2,Modulator_2,Modulator_1,Modulator_1]}
 {
  id=Random_2
  children [WavePlayer_10, WavePlayer_11, WavePlayer_12, WavePlayer_13, WavePlayer_14, WavePlayer_15, WavePlayer_16, WavePlayer_19, WavePlayer_20]
  properties {weights [1,1,1,1,1,1,1,1,1]}
 }
 {id=Random_3, children [WavePlayer_22, WavePlayer_21, WavePlayer_23], properties {weights [1,1,1]}}
 {id=WavePlayer_10, properties {soundWaveAssetPtr=&/Impact_Chrome_Large_01}}
 {id=WavePlayer_11, properties {soundWaveAssetPtr=&/Impact_Chrome_Large_02}}
 {id=WavePlayer_12, properties {soundWaveAssetPtr=&/Impact_Chrome_Large_03}}
 {id=WavePlayer_13, properties {soundWaveAssetPtr=&/Impact_Chrome_Large_04}}
 {id=WavePlayer_14, properties {soundWaveAssetPtr=&/Impact_Chrome_Large_05}}
 {id=WavePlayer_15, properties {soundWaveAssetPtr=&/Impact_Chrome_Large_06}}
 {id=WavePlayer_16, properties {soundWaveAssetPtr=&/Impact_Chrome_Large_07}}
 {id=WavePlayer_19, properties {soundWaveAssetPtr=&/Impact_Chrome_Large_08}}
 {id=WavePlayer_20, properties {soundWaveAssetPtr=&/Impact_Chrome_Large_09}}
 {id=WavePlayer_21, properties {soundWaveAssetPtr=&Impact_Chrome_Large_04}}
 {id=WavePlayer_22, properties {soundWaveAssetPtr=&Impact_Chrome_Large_02}}
 {id=WavePlayer_23, properties {soundWaveAssetPtr=&Impact_Chrome_Large_06}}
]
properties {
 attenuationSettings=/Game/Sounds/Attenuation/Fort_Bullet_Impact_Att
 maxDistance=5300
 soundClassObject=&Fort_Impacts
 volumeMultiplier=1.133
 concurrencyOverrides {maxCount=4, resolutionRule=PreventNew}
 concurrencySet [&RifleBulletImpact_Conc]
}
