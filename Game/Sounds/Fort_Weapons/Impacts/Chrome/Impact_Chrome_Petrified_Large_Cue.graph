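# Sound cue for large petrified chrome weapon impacts, using bullet-impact attenuation, the impacts sound class and rifle bullet-impact concurrency.
# A short random delay feeds a quality-level switch, whose first two slots pick randomly from nine wave variations and last two from a smaller three-wave subset.
# Each branch applies the same random pitch and volume modulation, with pitch only shifting upward.
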SoundCue /Game/Sounds/Fort_Weapons/Impacts/Chrome/Impact_Chrome_Petrified_Large_Cue
root=Delay_0
nodes [
 {id=Delay_0, children [QualityLevel_1], properties {delayMax=0.13, delayMin=0.07}}
 {id=Modulator_0, children [Random_0], properties {pitchMax=1.25, pitchMin=1, volumeMax=1.1, volumeMin=0.9}}
 {id=Modulator_3, children [Random_1], properties {pitchMax=1.25, pitchMin=1, volumeMax=1.1, volumeMin=0.9}}
 {id=QualityLevel_1, children [Modulator_0,Modulator_0,Modulator_3,Modulator_3]}
 {
  id=Random_0
  children [WavePlayer_0,WavePlayer_1,WavePlayer_2,WavePlayer_3,WavePlayer_4,WavePlayer_5,WavePlayer_6,WavePlayer_7,WavePlayer_8]
  properties {weights [1,1,1,1,1,1,1,1,1]}
 }
 {id=Random_1, children [WavePlayer_17, WavePlayer_18, WavePlayer_9], properties {weights [1,1,1]}}
 {id=WavePlayer_0, properties {soundWaveAssetPtr=&/Impact_Chrome_Petrified_Large_01}}
 {id=WavePlayer_1, properties {soundWaveAssetPtr=&/Impact_Chrome_Petrified_Large_02}}
 {id=WavePlayer_17, properties {soundWaveAssetPtr=&Impact_Chrome_Petrified_Large_01}}
 {id=WavePlayer_18, properties {soundWaveAssetPtr=&/Impact_Chrome_Petrified_Large_04}}
 {id=WavePlayer_2, properties {soundWaveAssetPtr=&/Impact_Chrome_Petrified_Large_03}}
 {id=WavePlayer_3, properties {soundWaveAssetPtr=&Impact_Chrome_Petrified_Large_04}}
 {id=WavePlayer_4, properties {soundWaveAssetPtr=&/Impact_Chrome_Petrified_Large_05}}
 {id=WavePlayer_5, properties {soundWaveAssetPtr=&/Impact_Chrome_Petrified_Large_06}}
 {id=WavePlayer_6, properties {soundWaveAssetPtr=&/Impact_Chrome_Petrified_Large_07}}
 {id=WavePlayer_7, properties {soundWaveAssetPtr=&/Impact_Chrome_Petrified_Large_08}}
 {id=WavePlayer_8, properties {soundWaveAssetPtr=&/Impact_Chrome_Petrified_Large_09}}
 {id=WavePlayer_9, properties {soundWaveAssetPtr=&Impact_Chrome_Petrified_Large_06}}
]
properties {
 attenuationSettings=/Game/Sounds/Attenuation/Fort_Bullet_Impact_Att
 maxDistance=5300
 soundClassObject=&Fort_Impacts
 volumeMultiplier=1.133
 concurrencyOverrides {maxCount=4, resolutionRule=PreventNew}
 concurrencySet [&RifleBulletImpact_Conc]
}
