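# Sound cue for small petrified chrome impacts, using bullet impact attenuation, the Fort_Impacts sound class and SMG bullet impact concurrency.
# A short random delay feeds a quality-level switch: the first two slots pick from eight wave variations, the last two from a reduced set of three.
# Each branch randomizes pitch and volume through a modulator over its equal-weight random pick.
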SoundCue /Game/Sounds/Fort_Weapons/Impacts/Chrome/Impact_Chrome_Petrified_Small_Cue
root=Delay_0
nodes [
 {id=Delay_0, children [QualityLevel_1], properties {delayMax=0.13, delayMin=0.07}}
 {id=Modulator_2, children [Random_2], properties {pitchMax=1.15, pitchMin=0.9, volumeMax=1, volumeMin=0.8}}
 {id=Modulator_3, children [Random_3], properties {pitchMax=1.15, pitchMin=0.9, volumeMax=1, volumeMin=0.8}}
 {id=QualityLevel_1, children [Modulator_2,Modulator_2,Modulator_3,Modulator_3]}
 {
  id=Random_2
  children [WavePlayer_10, WavePlayer_11, WavePlayer_12, WavePlayer_13, WavePlayer_14, WavePlayer_15, WavePlayer_16, WavePlayer_19]
  properties {weights [1,1,1,1,1,1,1,1]}
 }
 {id=Random_3, children [WavePlayer_21, WavePlayer_22, WavePlayer_20], properties {weights [1,1,1]}}
 {id=WavePlayer_10, properties {soundWaveAssetPtr=&/Impact_Chrome_Petrified_Small_01}}
 {id=WavePlayer_11, properties {soundWaveAssetPtr=&/Impact_Chrome_Petrified_Small_02}}
 {id=WavePlayer_12, properties {soundWaveAssetPtr=&/Impact_Chrome_Petrified_Small_03}}
 {id=WavePlayer_13, properties {soundWaveAssetPtr=&/Impact_Chrome_Petrified_Small_04}}
 {id=WavePlayer_14, properties {soundWaveAssetPtr=&/Impact_Chrome_Petrified_Small_05}}
 {id=WavePlayer_15, properties {soundWaveAssetPtr=&/Impact_Chrome_Petrified_Small_06}}
 {id=WavePlayer_16, properties {soundWaveAssetPtr=&/Impact_Chrome_Petrified_Small_07}}
 {id=WavePlayer_19, properties {soundWaveAssetPtr=&/Impact_Chrome_Petrified_Small_08}}
 {id=WavePlayer_20, properties {soundWaveAssetPtr=&Impact_Chrome_Petrified_Small_08}}
 {id=WavePlayer_21, properties {soundWaveAssetPtr=&Impact_Chrome_Petrified_Small_01}}
 {id=WavePlayer_22, properties {soundWaveAssetPtr=&Impact_Chrome_Petrified_Small_02}}
]
properties {
 attenuationSettings=/Game/Sounds/Attenuation/Fort_Bullet_Impact_Att
 maxDistance=5300
 soundClassObject=&Fort_Impacts
 volumeMultiplier=0.956
 concurrencyOverrides {maxCount=4, resolutionRule=PreventNew}
 concurrencySet [/Game/Sounds/Concurrency_Rules/SMGBulletImpact_Conc]
}
